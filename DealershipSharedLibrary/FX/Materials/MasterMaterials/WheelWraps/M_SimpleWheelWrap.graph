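# Masked, lit master material for wheel wrap effects on Niagara sprites, ribbons and mesh particles.
# A dynamic parameter switches the emissive between the multiplied main texture and a recolored version of its desaturated mask, via a shared emissive adjustment function.
# Opacity selects its source with a 0/1 parameter, scales by particle alpha and an optional wheel band from a rounded sine of V.
# The Opacity branch adds depth fade; Opacity Mask omits it.
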Material /DealershipSharedLibrary/FX/Materials/MasterMaterials/WheelWraps/M_SimpleWheelWrap
graph {
 properties {
  bAutomaticallySetUsageInEditor=false
  bUsedWithNiagaraMeshParticles=true
  bUsedWithNiagaraRibbons=true
  bUsedWithNiagaraSprites=true
  blendMode=BLEND_Masked
  usedShadingModels=Default Lit
 }
 nodes [
  n1 = TextureSampleParameter2D() {group=General, parameterName=Main Texture, texture=/DealershipSharedLibrary/FX/Textures/Diffuse/T_VCVFX_TriangleOP}
  n2 = Multiply(A=@n1.RGB, B=NamedRerouteUsage())
  n3 = NamedRerouteDeclaration(None=Saturate(None=Desaturation(None=@n2))) {name=Desaturated Mask, nodeColor=<rgba 1 0 0.6797 1>}
  n4 = NamedRerouteUsage()
  MaterialFunctionCall(Emissive=LinearInterpolate(
    A=@n2
    B=Multiply(A=@n3, B=@n4)
    Alpha=DynamicParameter() {defaultValue=<rgba 0 1 0 1>, paramNames ["IsPainted? 0=False 1=True", Param2, Param3, Param4]}."IsPainted? 0=False 1=True"
   )) {materialFunction=&MF_FX_EmissiveAdjustments}
  MaterialExpressionVertexColor_3 = ParticleColor()
  NamedRerouteDeclaration(None=@MaterialExpressionVertexColor_3.RGB) {name=Particle Color RGB, nodeColor=<rgba 0.3594 1 0 1>}
  n5 = Floor(None=ScalarParameter() {group=Opacity, parameterName="Use Alpha From Texture? 0=False 1=True"})
  n6 = Multiply(A=LinearInterpolate(A=NamedRerouteUsage(), B=NamedRerouteUsage(), Alpha=Saturate(None=@n5)), B=@MaterialExpressionVertexColor_3.A)
  n7 = ScalarParameter() {defaultValue=10, group=Opacity, parameterName=DepthFade Distance}
  n8 = Sine(None=TextureCoordinate()) {period=2}
  n9 = LinearInterpolate(B=Round(None=ComponentMask(None=@n8) {g=true}), Alpha=ScalarParameter() {group=Opacity, parameterName="Mask for Wheel? 0=False 1=True"}) {constA=1}
  NamedRerouteDeclaration(None=Multiply(
    A=MaterialFunctionCall(Opacity In=@n6, DepthFade Distance=@n7) {materialFunction=&MF_VCVFX_DepthFade}.Result
    B=@n9
   )) {name=Opacity, nodeColor=<rgba 0 1 1 1>}
  NamedRerouteDeclaration(None=Multiply(A=@n6, B=@n9)) {name=Opacity Mask, nodeColor=<rgba 0 0.7656 1 1>}
  NamedRerouteDeclaration(None=@n1.A) {name=Alpha From Texture, nodeColor=<rgba 0 0.3672 1 1>}
  NamedRerouteUsage()
  NamedRerouteUsage()
 ]
}
