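# Unlit translucent master material for wheel-wrap effects on Niagara mesh particles.
# Two panning texture layers, each with tiling, offset, centered scale and channel selection, are added for emissive (through MF_FX_EmissiveAdjustments) and opacity, with Layer 02 optional.
# Per-layer and global edge masks use smoothstepped distance from the V center, and particle color and alpha are exposed through reroutes.
# Layer 02 offers hue shift or a custom color, which uses only its red channel as grayscale.
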Material /DealershipSharedLibrary/FX/Materials/MasterMaterials/WheelWraps/M_Standard_WheelWrap
graph {
 properties {
  bAutomaticallySetUsageInEditor=false
  bUsedWithNiagaraMeshParticles=true
  blendMode=BLEND_Translucent
  shadingModel=MSM_Unlit
  usedShadingModels=Unlit
 }
 nodes [
  n1 = NamedRerouteUsage()
  n2 = NamedRerouteUsage()
  n3 = ScalarParameter() {defaultValue=2, group=Layer 01, parameterName=Layer 01 Tile U, sortPriority=3}
  n4 = ScalarParameter() {defaultValue=1, group=Layer 01, parameterName=Layer 01 Tile V, sortPriority=4}
  n5 = Add(
   A=Multiply(A=AppendVector(A=@n3, B=@n4), B=TextureCoordinate())
   B=AppendVector(
    A=ScalarParameter() {group=Layer 01, parameterName=Layer 01 Offset U, sortPriority=5}
    B=ScalarParameter() {group=Layer 01, parameterName=Layer 01 Offset V, sortPriority=6}
   )
  )
  n6 = AppendVector(
   A=ScalarParameter() {defaultValue=1, group=Layer 01, parameterName=Layer 01 Scale U, sortPriority=7}
   B=ScalarParameter() {defaultValue=1, group=Layer 01, parameterName=Layer 01 Scale V, sortPriority=8}
  )
  n7 = ScalarParameter() {defaultValue=2, group=Layer 01, parameterName=Layer 01 Speed U, sortPriority=1}
  n8 = ScalarParameter() {group=Layer 01, parameterName=Layer 01 Speed V, sortPriority=2}
  n9 = TextureSampleParameter2D(UVs=Panner(
    Coordinate=MaterialFunctionCall(UVs=@n5, Texture Scale=@n6) {materialFunction=&ScaleUVsByCenter}.UVs
    Speed=AppendVector(A=@n7, B=@n8)
   ) {speedX=2}) {group=Layer 01, parameterName=Layer 01 Texture, sortPriority=0, texture=/DealershipSharedLibrary/FX/Textures/Gradient_Masks/T_VCVFX_BeamLine}
  n10 = StaticComponentMaskParameter(None=@n9.RGB) {defaultR=true, group=Layer 01, parameterName=Layer 01 Channel Selector, sortPriority=9}
  n11 = StaticSwitchParameter(True=Multiply(A=@n10, B=@n2), False=@n10) {group=Layer 01, parameterName=Use Layer 01 Edge Mask?, sortPriority=10}
  n12 = Multiply(A=NamedRerouteUsage(), B=@n11)
  MaterialFunctionCall(Emissive=StaticSwitchParameter(True=Add(A=@n12, B=@n1), False=@n12) {defaultValue=true, group=General, parameterName=Use Layer 02?, sortPriority=100}) {materialFunction=&MF_FX_EmissiveAdjustments}
  n13 = Multiply(A=@n9.A, B=@n2)
  n14 = ScalarParameter() {
   controlType=Enumeration
   enumeration=/DealershipSharedLibrary/FX/Materials/Enums/Enum_DLR_TrueFalse
   group=Layer 01
   parameterName=UseAlphaFromText01
   sliderMax=1
  }
  n15 = Saturate(None=@n14)
  n16 = NamedRerouteDeclaration(None=Multiply(A=NamedRerouteUsage(), B=LinearInterpolate(A=@n11, B=@n13, Alpha=@n15) {constAlpha=1})) {name=Layer 01 Alpha, nodeColor=<rgba 1 0 0.9141 1>}
  n17 = NamedRerouteUsage()
  Multiply(
   A=StaticSwitchParameter(True=Add(A=@n16, B=@n17), False=@n16) {defaultValue=true, group=General, parameterName=Use Layer 02?, sortPriority=100}
   B=NamedRerouteUsage()
  )
  n18 = ScalarParameter() {defaultValue=0.3, group=Layer 02, parameterName=Layer 02 Mask Width, sortPriority=27}
  n19 = TextureCoordinate()
  n20 = SmoothStep(
   Min=Add(A=@n18, B=ScalarParameter() {defaultValue=0.05, group=Layer 02, parameterName=Layer 02 Ramp Width, sortPriority=28}) {constB=0.05}
   Max=@n18
   Value=Distance(A=ComponentMask(None=@n19) {g=true}, B=Constant() {r=0.5})
  ) {constMax=0.05}
  NamedRerouteDeclaration(None=LinearInterpolate(A=@n20, B=OneMinus(None=@n20), Alpha=ScalarParameter() {group=Layer 02, parameterName=Layer 02 Invert Mask, sliderMax=1, sortPriority=29})) {name=EdgeMask - Layer 02, nodeColor=<rgba 1 0 0.5859 1>}
  n21 = VectorParameter() {group=Layer 02, parameterName=Layer 02 Custom Color, defaultValue=<rgba 1 1 1 1>}
  n22 = MaterialFunctionCall(X=@n21.R, Y=@n21.R, Z=@n21.R) {materialFunction=&MakeFloat3}
  n23 = NamedRerouteUsage()
  n24 = LinearInterpolate(
   A=@n23
   B=MaterialFunctionCall(Hue Shift Percentage=ScalarParameter() {group=Layer 02, parameterName=HueShift Percentage, sortPriority=25}, Texture=@n23) {materialFunction=&HueShift}.Result
   Alpha=ScalarParameter() {group=Layer 02, parameterName=Use Layer 02 Hue Shift?, sliderMax=1, sortPriority=24}
  )
  n25 = ScalarParameter() {defaultValue=0.25, group=Layer 02, parameterName=Layer 02 Tile U, sortPriority=17}
  n26 = ScalarParameter() {defaultValue=0.75, group=Layer 02, parameterName=Layer 02 Tile V, sortPriority=18}
  n27 = Add(
   A=Multiply(A=AppendVector(A=@n25, B=@n26), B=TextureCoordinate())
   B=AppendVector(
    A=ScalarParameter() {group=Layer 02, parameterName=Layer 02 Offset U, sortPriority=19}
    B=ScalarParameter() {group=Layer 02, parameterName=Layer 02 Offset V, sortPriority=20}
   )
  )
  n28 = AppendVector(
   A=ScalarParameter() {defaultValue=1, group=Layer 02, parameterName=Layer 02 Scale U, sortPriority=21}
   B=ScalarParameter() {defaultValue=1, group=Layer 02, parameterName=Layer 02 Scale V, sortPriority=22}
  )
  n29 = ScalarParameter() {defaultValue=2, group=Layer 02, parameterName=Layer 02 Speed U, sortPriority=15}
  n30 = ScalarParameter() {group=Layer 02, parameterName=Layer 02 Speed V, sortPriority=16}
  n31 = TextureSampleParameter2D(UVs=Panner(
    Coordinate=MaterialFunctionCall(UVs=@n27, Texture Scale=@n28) {materialFunction=&ScaleUVsByCenter}.UVs
    Speed=AppendVector(A=@n29, B=@n30)
   ) {speedX=2}) {group=Layer 02, parameterName=Layer 02 Texture, sortPriority=14, texture=/DealershipSharedLibrary/FX/Textures/Noise/T_VCVFX_NoiseSpots_RGB}
  n32 = StaticComponentMaskParameter(None=@n31.RGB) {defaultR=true, group=Layer 02, parameterName=Layer 02 Channel Selector, sortPriority=23}
  n33 = StaticSwitchParameter(True=Multiply(A=@n32, B=NamedRerouteUsage()), False=@n32) {defaultValue=true, group=Layer 02, parameterName=Use Layer 02 Edge Mask?, sortPriority=26}
  NamedRerouteDeclaration(None=Multiply(A=StaticSwitchParameter(True=@n22.Result, False=@n24) {group=Layer 02, parameterName=Use Custom Color?}, B=@n33)) {name=Layer 02, nodeColor=<rgba 0 0.9062 1 1>}
  n34 = ScalarParameter() {defaultValue=0.3, group=Layer 01, parameterName=Layer 01 Mask Width, sortPriority=11}
  n35 = TextureCoordinate()
  n36 = SmoothStep(
   Min=Add(A=@n34, B=ScalarParameter() {defaultValue=0.05, group=Layer 01, parameterName=Layer 01 Ramp Width, sortPriority=12}) {constB=0.05}
   Max=@n34
   Value=Distance(A=ComponentMask(None=@n35) {g=true}, B=Constant() {r=0.5})
  ) {constMax=0.05}
  NamedRerouteDeclaration(None=LinearInterpolate(A=@n36, B=OneMinus(None=@n36), Alpha=ScalarParameter() {group=Layer 01, parameterName=Layer 01 Invert Mask?, sliderMax=1, sortPriority=13})) {name=EdgeMask - Layer 01, nodeColor=<rgba 1 0 0.5859 1>}
  n37 = ParticleColor()
  NamedRerouteDeclaration(None=@n37.RGB) {name=Particle Color, nodeColor=<rgba 0 1 0.1328 1>}
  NamedRerouteDeclaration(None=@n37.A) {name=Particle Alpha, nodeColor=<rgba 0 1 0.1328 1>}
  NamedRerouteDeclaration(None=Multiply(A=NamedRerouteUsage(), B=@n33)) {name=Layer 02 Alpha, nodeColor=<rgba 1 0 0.9141 1>}
  n38 = ScalarParameter() {defaultValue=0.275, group=General, parameterName=Global Mask Width}
  n39 = ScalarParameter() {defaultValue=0.05, group=General, parameterName=Global Mask Ramp}
  n40 = ComponentMask(None=TextureCoordinate()) {g=true}
  n41 = Constant() {r=0.5}
  NamedRerouteDeclaration(None=SmoothStep(Min=Add(A=@n38, B=@n39) {constB=0.05}, Max=@n38, Value=Distance(A=@n40, B=@n41)) {constMax=0.05}) {name=GlobalEdgeMask, nodeColor=<rgba 1 0 0.5859 1>}
 ]
}
