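# Material function for translucent materials that masks opacity with a manual depth test against the scene.
# It adds an optional Bias, defaulting to -1, to PixelDepth and compares the result with SceneDepth through an If node.
# The hard binary result multiplies the Opacity input, zeroing it where the biased pixel lies behind scene depth and passing it through otherwise.
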MaterialFunction /DealershipSharedLibrary/FX/Materials/MaterialFunctions/Mutable/MF_TranslucencySortDepth_PSY
graph {nodes [
  n1 = FunctionInput() {bUsePreviewValueAsDefault=true, inputName=Bias, inputType=FunctionInput_Scalar, previewValue=<-1 0 0 1>}
  n2 = Add(A=PixelDepth().R, B=@n1)
  n3 = ComponentMask(None=SceneDepth().R) {r=true}
  n4 = Constant()
  n5 = Constant() {r=1}
  FunctionOutput(None=Multiply(A=FunctionInput() {inputName=Opacity, inputType=FunctionInput_Scalar}, B=If(A=@n2, B=@n3, "A > B"=@n4, "A == B"=@n5, "A < B"=@n5)))
 ]}
