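# Translucent unlit master material for stylized smoke on Niagara mesh particles, filed under WheelWraps.
# Fakes toon lighting by blending camera-relative and sky-atmosphere-light normal dot products, posterizing the result and multiplying it with an optionally posterized shadow term and particle color.
# Panned, distorted cell noise plus Fresnel erodes particle alpha, with optional depth fade and dither variants.
# Dynamic parameters drive normal-based displacement, shadow strength, animation time and noise-based emissive sparkles.
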Material /DealershipSharedLibrary/FX/Materials/MasterMaterials/WheelWraps/M_Stylized_Smoke2
graph {
 properties {
  bUsedWithNiagaraMeshParticles=true
  blendMode=BLEND_Translucent
  pixelDepthOffsetMode=PDOM_Legacy
  shadingModel=MSM_Unlit
  translucencyPass=MTP_AfterMotionBlur
  usedShadingModels=Unlit
 }
 nodes [
  n1 = ParticleColor()
  n2 = CameraVectorWS()
  n3 = Transform(None=@n2) {transformSourceType=TRANSFORMSOURCE_World, transformType=TRANSFORM_Camera}
  n4 = ScalarParameter() {
   defaultValue=0.5
   desc=Adjustment to highlight one side of the contour based on the camera vector.
   group=Shading
   parameterName=Shading Adjustment
  }
  n5 = PixelNormalWS()
  n6 = DotProduct(A=LinearInterpolate(A=Multiply(A=@n3), B=@n2, Alpha=@n4), B=@n5)
  n7 = DynamicParameter() {parameterIndex=1, defaultValue=<rgba -100 1 1 -1>, paramNames [Displacement, ShadowStrength, PlayAnimTime, Dissolve]}
  n8 = Add(A=Multiply(A=@n5, B=@n7.PlayAnimTime) {constB=0.5}) {constB=0.5}
  n9 = NamedRerouteUsage()
  n10 = Add(A=Multiply(A=@n8, B=OneMinus(None=@n9)), B=@n8)
  n11 = Multiply(A=SkyAtmosphereLightDirection(), B=ScalarParameter() {defaultValue=0.333, group=Shading, parameterName=SkyAtmosphereLightDir_BlendRate}) {constB=0.333}
  n12 = AppendVector(A=ComponentMask(None=@n11) {g=true, r=true}, B=LinearInterpolate(A=ComponentMask(None=@n11) {b=true}, B=Constant() {r=1}) {constAlpha=0.7})
  n13 = Clamp(None=LinearInterpolate(A=OneMinus(None=@n6), B=DotProduct(A=@n10, B=@n12)) {constAlpha=0.6}) {maxDefault=2, minDefault=0.1}
  n14 = ScalarParameter() {defaultValue=2.7, group=Shading, parameterName=Shadow Posterize}
  n15 = Multiply(
   A=NamedRerouteUsage()
   B=DynamicParameter() {defaultValue=<rgba -100 1 1 -1>, paramNames [Displacement, ShadowStrength, PlayAnimTime, Dissolve]}.ShadowStrength
  )
  n16 = OneMinus(None=Power(Base=Saturate(None=@n15), Exp=ScalarParameter() {defaultValue=1.5, group=Shading, parameterName=Shadow Power}))
  n17 = Multiply(
   A=MaterialFunctionCall!(Posterize=ScalarParameter() {defaultValue=1.5, group=Shading, parameterName=Posterize}, Input=@n13) {materialFunction=/DealershipSharedLibrary/FX/Materials/MaterialFunctions/MF_Simple_Posterize}.Result
   B=StaticSwitch(
    True=MaterialFunctionCall(Posterize=@n14, Input=@n16) {materialFunction=&MF_Simple_Posterize}.Result
    False=@n16
    Value=StaticBoolParameter() {defaultValue=true, parameterName=Enable Posterize Surface Shadow}
   )
  )
  n18 = Multiply(A=@n1.RGB, B=@n17)
  Add(
   A=Multiply(
    A=MaterialFunctionCall() {materialFunction=/Game/Packages/Fortress_SharedMaterials/Base_Material_Functions/SkylightColorAndIntensity.SkyLightColorAndIntensity}.Color
    B=@n18
   )
   B=@n18
  )
  MaterialFunctionCall!(Bias, Opacity=NamedRerouteUsage()) {materialFunction=/DealershipSharedLibrary/FX/Materials/MaterialFunctions/Mutable/MF_TranslucencySortDepth_PSY}
  n19 = ScalarParameter() {defaultValue=0.01, desc=How much consider the sky atmosphere lighting, group=Shading, parameterName=SkyAtmosphere LightScale}
  n20 = Multiply(A=@n19, B=SkyAtmosphereLightIlluminance!(World Position))
  Multiply(A=Saturate(None=Add(A=@n20, B=@n19) {constB=0.1}), B=Multiply(A=Add(A=@n20, B=@n1.RGB), B=@n17))
  n21 = DynamicParameter() {parameterIndex=1, defaultValue=<rgba -100 1 1 -1>, paramNames [Displacement, ShadowStrength, PlayAnimTime, Dissolve]}
  n22 = Subtract(A=NamedRerouteUsage(), B=ParticlePositionWS())
  n23 = TextureSample(UVs=Multiply(A=Multiply(A=@n21.ShadowStrength, B=@n22)) {constB=0.5}) {samplerType=SAMPLERTYPE_LinearColor, texture=/DealershipSharedLibrary/FX/Textures/VirtualTextures/VT_VCVFX_Voronoi}
  n24 = QualitySwitch(Default=@n23.R, Low=@n23.R, High=@n23.R, Medium=@n23.R)
  n25 = Multiply(A=@n24, B=@n24)
  NamedRerouteDeclaration(None=Multiply(A=@n25, B=@n25)) {name=Voronoi Noise Result 2, nodeColor=<rgba 1 0 0.0703 1>}
  n26 = Constant3Vector() {constant=<rgba -0.006 0.01 1 0>}
  n27 = Add(
   A=Multiply(
    A=DynamicParameter() {defaultValue=<rgba -100 1 1 -1>, paramNames [Displacement, ShadowStrength, PlayAnimTime, Dissolve]}.PlayAnimTime
    B=Normalize(VectorInput=@n26)
   )
   B=Multiply(A=ParticleRandom()) {constB=100}
  )
  n28 = WorldPosition()
  NamedRerouteDeclaration(None=Add(A=Multiply() {constA=2, constB=4}, B=Add(A=@n27, B=@n28.XYZ))) {name=Volume Noise UVW, nodeColor=<rgba 0 1 0.7188 1>}
  n29 = Multiply(A=VertexNormalWS(), B=NamedRerouteUsage())
  n30 = DynamicParameter() {defaultValue=<rgba -100 1 1 -1>, paramNames [Displacement, ShadowStrength, PlayAnimTime, Dissolve]}
  NamedRerouteDeclaration(None=StaticSwitch(
    True=Multiply(A=@n29, B=@n30.Displacement)
    False=Constant()
    Value=StaticBoolParameter() {defaultValue=true, parameterName=Use Displacement}
   )) {name=Displacement Result, nodeColor=<rgba 0 0.6484 1 1>}
  n31 = ScalarParameter() {defaultValue=0.2, parameterName=Tile}
  n32 = TextureCoordinate()
  n33 = ScalarParameter() {defaultValue=0.2, parameterName=Speed}
  n34 = Time()
  n35 = TextureSample(UVs=Panner(Coordinate=Multiply(A=@n31, B=@n32), Time=Multiply(A=@n33, B=@n34)) {speedY=0.25}) {texture=/DealershipSharedLibrary/FX/Textures/Noise/T_VCVFX_CellNoise01_A}
  n36 = TextureSample(UVs=Multiply(A=ScalarParameter() {defaultValue=0.2, parameterName=Intensity}, B=ComponentMask(None=@n35.RGB) {g=true, r=true})) {texture=&T_VCVFX_CellNoise01_A}
  n37 = Divide(A=Add(A=OneMinus(None=@n36.RGB), B=Fresnel() {exponent=1})) {constB=2.5}
  n38 = Clamp(None=Subtract(A=ParticleColor().A, B=Clamp(None=@n37)))
  n39 = ScalarParameter() {defaultValue=100, parameterName=Depth Fade}
  NamedRerouteDeclaration(None=StaticSwitch(
    True=MaterialFunctionCall(Opacity In=@n38, DepthFade Distance=@n39) {materialFunction=&MF_VCVFX_DepthFade}.Result
    False=@n38
    Value=StaticBoolParameter() {parameterName=Use Depth Fade when Translucent}
   )) {name=Opacity Result, nodeColor=<rgba 1 0.8203 0 1>}
  NamedRerouteDeclaration(None=StaticSwitch(
    True=MaterialFunctionCall(Alpha Threshold=@n38) {materialFunction=&DitherTemporalAA}.Result
    False=@n38
    Value=StaticBoolParameter() {defaultValue=true, parameterName=Use Dither TAA}
   )) {name=Opacity Mask Result, nodeColor=<rgba 0 0.25 1 1>}
  n40 = ComponentMask(None=TextureSample(UVs=Panner() {speedY=0.005}) {texture=/DealershipSharedLibrary/FX/Textures/Noise/T_VCVFX_Noise01_A}.RGB) {g=true, r=true}
  n41 = Multiply(
   A=TextureSample(UVs=TextureCoordinate() {uTiling=3, vTiling=3}) {texture=/DealershipSharedLibrary/FX/Textures/Noise/T_VCVFX_NoiseGeneric01Pack_RGB}.B
   B=TextureSample(UVs=Multiply(A=@n40) {constB=0.0625}) {texture=&T_VCVFX_NoiseGeneric01Pack_RGB}.B
  )
  n42 = Multiply(
   A=Multiply(A=Multiply(A=@n41) {constB=4096}, B=@n1.RGB)
   B=DynamicParameter() {parameterIndex=2, defaultValue=<rgba 0 1 1 1>, paramNames [Sparkles, "", "", ""]}.Sparkles
  )
  NamedRerouteDeclaration(None=MaterialFunctionCall(Emissive=Add(A=@n42, B=@n18)) {materialFunction=&MF_FX_EmissiveAdjustments}.Result) {name=Emissive Result, nodeColor=<rgba 0 0.5781 1 1>}
  NamedRerouteDeclaration(None=Multiply(A=@n1.RGB, B=@n17)) {name=Base Color Result, nodeColor=<rgba 0.0312 1 0 1>}
  NamedRerouteDeclaration(None=Transform(None=@n24)) {name=Voronoi Noise Result 1, nodeColor=<rgba 0.8281 1 0 1>}
  NamedRerouteUsage()
  NamedRerouteUsage()
  NamedRerouteUsage()
  MaterialExpressionRemap_0 = MaterialXRemap!(None=@n13, InputLow, InputHigh, TargetLow, TargetHigh) {inputHighDefault=2, inputLowDefault=0.1}
 ]
}
